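# Translucent, unlit, two-sided material for the energy blade effect on the HighTower Tomato pickaxe.
# Color is a vertical two-color gradient, tinted by particle color and boosted to HDR emissive.
# Opacity comes from panning electricity noise, normal-map distorted except on mobile, shaped by sine edge falloffs, depth fade and view-angle fade.
# Panning cellular-noise dots erode opacity, mostly toward the U edges, with optional particle-alpha and vertex-red masking.
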Material /BRCosmetics/Weapons/FORT_Melee/Pickaxe_HighTower_Tomato/FX/M_HighTower_Tomato_EnergyBlade
graph {
 properties {
  bUsedWithMeshDeformer=true
  bUsedWithNiagaraMeshParticles=true
  bUsedWithSkeletalMesh=true
  blendMode=BLEND_Translucent
  ditherOpacityMask=true
  opacityMaskClipValue=0.333
  pixelDepthOffsetMode=PDOM_Legacy
  shadingModel=MSM_Unlit
  twoSided=true
  usedShadingModels=Unlit
 }
 nodes [
  n1 = TextureCoordinate()
  n2 = Power(Base=OneMinus(None=ComponentMask(None=@n1) {g=true}), Exp=ScalarParameter() {defaultValue=4, parameterName=Param_1})
  n3 = LinearInterpolate(
   A=VectorParameter() {group=Color, parameterName=ColorLow, sortPriority=3, defaultValue=<rgba 0 0.6667 1 1>}.RGB
   B=VectorParameter() {group=Color, parameterName=ColorHigh, sortPriority=4, defaultValue=<rgba 0 0.0333 1 1>}.RGB
   Alpha=Saturate(None=Multiply(A=@n2) {constB=2})
  )
  n4 = ParticleColor()
  MaterialFunctionCall(Emissive=MaterialFunctionCall(
    LDR Input=Saturate(None=@n3)
    HDR Tint=Multiply(A=@n3, B=@n4.RGB) {constA=5}
    HDR Intensity Multiplier=ScalarParameter() {defaultValue=5, group=Color, parameterName=HDR Intensity Multiplier, sortPriority=7}
   ) {materialFunction=&DeriveHDRfromLDR}.RemappedRange) {materialFunction=&MF_FX_EmissiveAdjustments}
  n5 = TextureCoordinate() {uTiling=0.1}
  n6 = ComponentMask(None=VectorParameter() {group=Distortion, parameterName=Distortion UV Scale, sortPriority=3, defaultValue=<rgba 1 1 0 0>}.RGB) {g=true, r=true}
  n7 = Time()
  n8 = ScalarParameter() {defaultValue=5, group=Panner Speed, parameterName=Speed}
  n9 = TextureSampleParameter2D(UVs=Panner(Coordinate=Multiply(A=@n5, B=@n6), Time=Multiply(A=@n7, B=@n8)) {speedY=1}) {
   group=Distortion
   parameterName=Distortion Normal
   samplerType=SAMPLERTYPE_Normal
   sortPriority=2
   texture=/Game/Effects/Fort_Effects/Textures/Tile/noise_flamy_Tile_NRM
  }
  n10 = MaterialFunctionCall(Normal=@n9.RGB, Flatness=ScalarParameter() {group=Distortion, parameterName=UV_Normal_Flatness, sortPriority=4}) {materialFunction=&FlattenNormal}
  n11 = TextureCoordinate()
  n12 = VectorParameter() {group=Fire Shape, parameterName=Fire_MacroShape_UVScale, sortPriority=5, defaultValue=<rgba 1 1 0 0>}
  n13 = ComponentMask(None=@n12.RGB) {g=true, r=true}
  n14 = Multiply(A=@n11, B=@n13)
  n15 = Divide(
   A=LinearInterpolate(A=ComponentMask(None=@n10.Result) {g=true, r=true}, Alpha=ComponentMask(None=@n14) {g=true}) {constB=0}
   B=ScalarParameter() {defaultValue=3, group=Distortion, parameterName=UV_Distortion_Divsor, sortPriority=1}
  ) {constB=20}
  n16 = TextureCoordinate() {coordinateIndex=1, uTiling=0.8, vTiling=0.9}
  n17 = ComponentMask(None=VectorParameter() {group=Fire Shape, parameterName=Fire_Micro_UVScale, sortPriority=8, defaultValue=<rgba 1 5 0 0>}.RGB) {g=true, r=true}
  n18 = Panner(
   Coordinate=MaterialFunctionCall(XY=Multiply(A=@n16, B=@n17)) {materialFunction=&Swizzle}.YX
   Time=Multiply(A=Time(), B=ScalarParameter() {defaultValue=5, group=Panner Speed, parameterName=Speed})
  ) {speedX=1.2, speedY=0.2}
  n19 = TextureSampleParameter2D(UVs=ShadingPathSwitch(Default=Add(A=@n15, B=@n18), Mobile=@n18) {desc=Migrated from FeatureLevelSwitch}) {
   group=Fire Shape
   parameterName=Fire Micro Texture
   samplerType=SAMPLERTYPE_Masks
   sortPriority=6
   texture=/Game/Characters/Player/Male/Medium/Bodies/M_MED_Storm_Soldier/Textures/ElectricityNoise_M
  }
  n20 = ChannelMaskParameter(None=@n19.RGBA) {group=Fire Shape, parameterName=Fire Micro Mask, sortPriority=7}
  n21 = MaterialFunctionCall(Float2=Multiply(A=@n11, B=@n13)) {materialFunction=&BreakOutFloat2Components}
  n22 = Multiply(A=@n21.R) {constB=0.5}
  n23 = ScalarParameter() {defaultValue=9.1, group=Fire Shape, parameterName=Edge Fade Value Top, sortPriority=3}
  n24 = ScalarParameter() {defaultValue=-3, group=Fire Shape, parameterName=Edge Fade Value Bottom, sortPriority=4}
  n25 = Multiply(A=Power(Base=Sine(None=@n22), Exp=LinearInterpolate(A=@n23, B=@n24, Alpha=@n21.G) {constA=6, constB=2}) {constExponent=4}, B=@n21.G)
  n26 = Multiply(
   A=Power(Base=Multiply(A=@n20.R, B=@n25), Exp=LinearInterpolate(Alpha=@n21.G) {constA=2})
   B=ScalarParameter() {defaultValue=300, group=Opacity, parameterName=OpacityMaskMultiplier, sortPriority=1}
  ) {constB=3}
  n27 = ScalarParameter() {defaultValue=20, parameterName=Fade Distance}
  n28 = ScalarParameter() {defaultValue=2, parameterName=Fade Sensitivity}
  n29 = ComponentMask(None=TextureCoordinate()) {r=true}
  n30 = Multiply(
   A=Multiply(
    A=DepthFade(Opacity=@n26, FadeDistance=@n27)
    B=MaterialFunctionCall("FadeSensitivity (optional)"=@n28) {materialFunction=&FadeBasedOnViewAngle}.Result
   )
   B=Power(Base=Sine(None=@n29) {period=2}, Exp=ScalarParameter() {defaultValue=2})
  )
  n31 = VectorParameter() {parameterName=Dot UV Scale, defaultValue=<rgba 1 10 0 0>}
  n32 = VectorParameter() {parameterName=Dot Pan Speed, defaultValue=<rgba 1 5 0 0>}
  n33 = LinearInterpolate(
   A=@n32.R
   B=Multiply(A=@n32.R) {constB=-1}
   Alpha=MaterialFunctionCall() {materialFunction=&DebugTimeSine}.Result
  )
  n34 = Panner(
   Coordinate=Multiply(A=TextureCoordinate(), B=ComponentMask(None=@n31.RGB) {g=true, r=true})
   Speed=LinearInterpolate(
    A=ComponentMask(None=@n32.RGB) {g=true, r=true}
    B=AppendVector(A=@n33, B=@n32.G)
    Alpha=ScalarParameter() {defaultValue=1, parameterName=Vary Dot Panning}
   )
  ) {bFractionalPart=true, speedX=1, speedY=5}
  n35 = Multiply(A=Power(
    Base=TextureSample(UVs=@n34) {samplerType=SAMPLERTYPE_Masks, texture=/Game/Global/Textures/Default/Noise/T_CellularNoise}.R
    Exp=ScalarParameter() {defaultValue=5, parameterName=Param_2}
   )) {constB=10}
  n36 = ComponentMask(None=TextureCoordinate()) {r=true}
  n37 = Multiply(A=@n35, B=OneMinus(None=Sine(None=@n36) {period=2}))
  n38 = ScalarParameter() {defaultValue=1, parameterName=Remove Dots from center}
  n39 = LinearInterpolate(B=Saturate(None=LinearInterpolate(A=@n35, B=@n37, Alpha=@n38)), Alpha=ScalarParameter() {defaultValue=1, parameterName=Use Dot Elements})
  n40 = ParticleColor()
  n41 = ScalarParameter() {parameterName=Use particle Alpha}
  n42 = VertexColor()
  Multiply(
   A=Multiply(A=LinearInterpolate(A=@n30, Alpha=@n39) {constB=0}, B=LinearInterpolate(B=@n40.A, Alpha=@n41) {constA=1})
   B=LinearInterpolate(A=LinearInterpolate(Alpha=@n42.R) {constA=1, constB=0}, Alpha=ScalarParameter() {parameterName=EmissiveCap})
  )
 ]
}
